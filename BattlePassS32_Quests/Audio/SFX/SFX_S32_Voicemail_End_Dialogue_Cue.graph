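# Sound cue that plays the Season 32 Battle Pass quest "Voicemail End" dialogue.
# A single dialogue player node plays the dialogue wave with the audio log VO speaker and no targets.
# It uses shared audio log attenuation and the narrator VO sound class.
# A pre-effect bus send feeds a quest VO envelope follower bus.
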SoundCue /BattlePassS32_Quests/Audio/SFX/SFX_S32_Voicemail_End_Dialogue_Cue
root=DialoguePlayer_0
nodes [{
  id=DialoguePlayer_0
  properties {dialogueWaveParameter {
    dialogueWave=&/SFX_S32_Voicemail_End_Dialogue
    context {speaker=/BattlePassBase/Audio/VO/AudioLogVO_DV, targets []}
   }}
 }]
properties {
 attenuationSettings=/BattlePassBase/Audio/SFX/Fort_AudioLogs_Att
 maxDistance=2097152
 soundClassObject=/Game/Sounds/FortSoundClassesAndModes/Fort_VO_Narrator
 volumeMultiplier=1
 preEffectBusSends [{
   audioBus=/Game/Sounds/Modulation/EnvelopeFollowers/BR_Boss/AB_Quest_VO_EnvFollow
   maxSendDistance=8000
   minSendDistance=0
   minSendLevel=1
   soundSourceBus=null
  }]
}
